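# Niagara system for the looping LunchBox character body effect: flickering electric arcs in warm yellow, with matching lights.
# FN_BoneFlow places particles along skinned bones: ribbon arcs and lights follow the right-arm bone chain and User.SkeletalMeshLeft, while electric sprites follow User.SkeletalMesh.
# Bursts at random times in two-second loops drive arcs and lights, while sprites spawn continuously.
# Ribbon width adapts to LOD distance, and User.bisHandHidden zeroes it and switches sprite and light visibility tags.
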NiagaraSystem /BRCosmetics/Effects/Fort_Effects/Effects/Characters/Athena_Parts/LunchBox/NS_LunchBox_Body_Loop4
dependencies {
 usedDataInterfaces [&NiagaraDataInterfaceCurve, &NiagaraDataInterfaceColorCurve, &NiagaraDataInterfaceVector2DCurve]
 usedDynamicInputs []
 usedModules [&SystemState]
}
emitters {
 BoneFlow_ArcLarge001 {
  emitterSpawnScript [
   /SetVariables(Emitter.RandomSeed:NiagaraInt32=/Niagara/DynamicInputs/Random/RandomInt(Randomness Mode:ENiagaraRandomnessMode=::NewEnumerator0, Random Seed:NiagaraInt32=0, Fixed Random Seed:NiagaraBool=false)) {isSetParameters=true, module=SetVariables_#1}
   /Game/Effects/Niagara/Modules/Partition/FN_PartitionAttributes(Number of Bone Sets=1, Particles Per Partition=8)
  ]
  emitterUpdateScript [
   &EmitterState(Life Cycle Mode=::NewEnumerator1, Loop Duration=2, Scalability Mode=::NewEnumerator1, UseLoopDelay=false)
   &SpawnBurst_Instantaneous(
    Spawn Count=$Emitter.FN_PartitionAttributes.Spawn Total
    Spawn Time=&RandomRangeFloat(Minimum=0, Maximum=1)
    Use Spawn Probability=false
   )
  ]
  particleSpawnScript [
   /Game/Effects/Niagara/Modules/Partition/FN_Particle_partition_v5/FN_Particle_Partition_v5(Enable Ribbon ID=true, Seed:NiagaraInt32=$Emitter.RandomSeed, Shuffle Partition ID=true)
   /Niagara/Modules/Spawn/Initialization/InitializeParticle(
    Lifetime=5
    Module.Write Color=true
    Module.Write Lifetime=false
    Module.Write Mass=true
    Module.Write Material Random=true
    Module.Write Position=false
    Module.Write Scale=false
    Module.Write SpriteRotation=false
    Module.Write SpriteSize=true
    Position=&SimulationPosition()
    Sprite Size=<5 5>
   )
   /Game/Effects/Niagara/Modules/BoneFlow/FN_Initialize_BoneFlow_Attributes_v2(
    Angle:NiagaraFloat=&Add_Float(
     A=&Multiply_Float(
      A=&FloatFromCurve(
       FloatCurve=DataInterface() {propertyValues {Curve {keys [
           {arriveTangent=0.0406, interpMode=RCIM_Cubic, leaveTangent=0.0406, tangentMode=RCTM_User, time=0, value=0}
           <at 0.6 0>
           {arriveTangent=-0.4332, interpMode=RCIM_Cubic, leaveTangent=-0.4332, tangentMode=RCTM_User, time=1, value=-0.3}
          ]}}}
       CurveIndex=$Particles.Partition.NormalizedID
      )
      B=&Multiply_Float(A=$Particles.Partition.NormalizedID, B=10)
     )
     B=0.2
    )
    Radius:NiagaraFloat=&RandomRangeFloat(Minimum=10, Maximum=15)
    T:NiagaraFloat=&UniformRangedFloat(Minimum=0, Maximum=1)
   )
   &InitializeRibbon(
    Color=<rgba 6 4.9108 0 1>
    Lifetime=&UniformRangedFloat(Minimum=0.2, Maximum=0.5)
    Module.Write Color=true
    Module.Write Lifetime=true
    Module.Write Mass=true
    Module.Write Material Random=true
    Module.Write Position=true
    Module.Write RibbonTwist=false
    Module.Write RibbonWidth=true
    Ribbon Width=&RandomRangeFloat(Minimum=16, Maximum=20)
   )
  ]
  particleUpdateScript [
   &ParticleState()
   &FN_BoneFlow(
    Angle:NiagaraFloat=$Particles.Initial.FN_Initialize_BoneFlow_Attributes_v2.Angle
    Height:NiagaraFloat=$Particles.Initial.Partition.NormalizedID
    Interpolate Mesh Sampling=false
    Radius:NiagaraFloat=&Add_Float(A=$Particles.Initial.FN_Initialize_BoneFlow_Attributes_v2.Radius, B=0)
    SkeletalMesh=DataInterface() {propertyValues {
      DefaultMesh=null
      ExcludeBoneName=root
      OptionalComponentClass=null
      PreviewMesh=&M_Med_Soldier_04_Preview
      SkinningMode=SkinOnTheFly
      SoftSourceActor=null
      SourceMode=Default
      UvSetIndex=0
      WholeMeshLOD=-1
      bExcludeBone=false
      bReadDeformedGeometry=true
      bRequireCurrentFrameData=true
      ComponentTags []
      FilteredBones [spine_05,upperarm_r,lowerarm_r]
      FilteredSockets []
      MeshUserParameter {parameter {name=None, typeDefHandle {registeredTypeIndex=1}, varData []}}
      SamplingRegions []
     }}
    Write Position Attribute=true
   )
   &Color(
    Color=&ColorFromCurve(ColorCurve=DataInterface() {propertyValues {
       AlphaCurve {keys [<at 0 0 RCIM_Cubic>,<at 0.1239 1>,<at 0.346 1>,<at 1 0 RCIM_Cubic>]}
       BlueCurve {keys [<at -0.0056 6>,<at 0.2609 0.2>]}
       GreenCurve {keys [<at -0.0056 23.0295>,<at 0.2609 0.7676>]}
       RedCurve {keys [<at -0.0056 30>,<at 0.2609 1>]}
      }})
    Scale Alpha=&FloatFromCurve(FloatCurve=DataInterface() {propertyValues {Curve {keys [<at 0.1282 1.0115>,<at 0.2455 0.0123>,<at 0.3389 0.9565>,<at 0.4758 0.9657>,<at 0.591 0.0032>,<at 0.6388 1.0115>,<at 0.8061 1.0207>,<at 1 0>]}}})
    Scale Color=&VectorFromFloat(Value=&FloatFromCurve(FloatCurve=DataInterface() {propertyValues {Curve {keys [<at 0 2>,<at 0.2994 -0.2627>,<at 0.573 2>,<at 0.8425 0.0123>]}}}))
   )
   &CameraOffset(Camera Offset Amount=25)
   &ScaleRibbonWidth(Ribbon Width Scale=&FloatFromCurve(
     FloatCurve=DataInterface() {propertyValues {Curve {keys [<at 0 0.35 RCIM_Constant>,<at 200 1 RCIM_Constant>]}}}
     CurveIndex=&ClampFloat(Float=$Engine.Owner.LODDistance, Min=20, Max=200)
    ))
   &DynamicMaterialParameters(
    Index 0 Param 1=1
    Index 0 Param 2=1
    Index 0 Param 3=&FloatFromCurve(
     FloatCurve=DataInterface() {propertyValues {Curve {keys [<at 0 2 RCIM_Constant>,<at 200 0.8 RCIM_Constant>]}}}
     CurveIndex=&ClampFloat(Float=$Engine.Owner.LODDistance, Max=200)
    )
    Module.Param0WriteEnabled=false
    Module.Param1WriteEnabled=false
    Module.Param2WriteEnabled=true
    Module.Param3WriteEnabled=false
   ) {module=DynamicMaterialParameters001}
   &ScaleRibbonWidth(Ribbon Width Scale=&MakeCustomFloatFromBool(Bool=$User.bisHandHidden, True Float=0, False Float=1)) {module=ScaleRibbonWidth001}
  ]
  rendererClasses [&NiagaraRibbonRendererProperties]
  renderers [{rendererClass=&NiagaraRibbonRendererProperties}]
 }
 BoneFlow_ArcLarge_Left {
  emitterSpawnScript [
   /SetVariables(Emitter.RandomSeed:NiagaraInt32=&RandomInt(Randomness Mode:ENiagaraRandomnessMode=::NewEnumerator0, Random Seed:NiagaraInt32=0, Fixed Random Seed:NiagaraBool=false)) {isSetParameters=true, module=SetVariables_#1}
   &FN_PartitionAttributes(Number of Bone Sets=1, Particles Per Partition=8)
  ]
  emitterUpdateScript [
   &EmitterState(Life Cycle Mode=::NewEnumerator1, Loop Duration=2, Scalability Mode=::NewEnumerator1, UseLoopDelay=false)
   &SpawnBurst_Instantaneous(
    Spawn Count=$Emitter.FN_PartitionAttributes.Spawn Total
    Spawn Time=&RandomRangeFloat(Minimum=0, Maximum=1)
    Use Spawn Probability=false
   )
  ]
  particleSpawnScript [
   &FN_Particle_Partition_v5(Enable Ribbon ID=true, Seed:NiagaraInt32=$Emitter.RandomSeed, Shuffle Partition ID=true)
   &InitializeParticle(
    Lifetime=5
    Module.Write Color=true
    Module.Write Lifetime=false
    Module.Write Mass=true
    Module.Write Material Random=true
    Module.Write Position=false
    Module.Write Scale=false
    Module.Write SpriteRotation=false
    Module.Write SpriteSize=true
    Position=&SimulationPosition()
    Sprite Size=<5 5>
   )
   &FN_Initialize_BoneFlow_Attributes_v2(
    Angle:NiagaraFloat=&Add_Float(
     A=&Multiply_Float(
      A=&FloatFromCurve(FloatCurve=DataInterface() {propertyValues {Curve {keys [<at 0 0>,<at 0.6 0>,<at 1 0.3>]}}}, CurveIndex=$Particles.Partition.NormalizedID)
      B=&Multiply_Float(A=$Particles.Partition.NormalizedID, B=10)
     )
     B=0.3
    )
    Radius:NiagaraFloat=&RandomRangeFloat(Minimum=10, Maximum=15)
    T:NiagaraFloat=&UniformRangedFloat(Minimum=0, Maximum=1)
   )
   &InitializeRibbon(
    Color=<rgba 6 4.9108 0 1>
    Lifetime=&UniformRangedFloat(Minimum=0.2, Maximum=0.5)
    Module.Write Color=true
    Module.Write Lifetime=true
    Module.Write Mass=true
    Module.Write Material Random=true
    Module.Write Position=true
    Module.Write RibbonTwist=false
    Module.Write RibbonWidth=true
    Ribbon Width=&RandomRangeFloat(Minimum=16, Maximum=20)
   )
  ]
  particleUpdateScript [
   &ParticleState()
   &FN_BoneFlow(
    Angle:NiagaraFloat=$Particles.Initial.FN_Initialize_BoneFlow_Attributes_v2.Angle
    Height:NiagaraFloat=$Particles.Initial.Partition.NormalizedID
    Interpolate Mesh Sampling=false
    Radius:NiagaraFloat=&Add_Float(A=$Particles.Initial.FN_Initialize_BoneFlow_Attributes_v2.Radius, B=0)
    SkeletalMesh:NiagaraDataInterfaceSkeletalMesh=$User.SkeletalMeshLeft
    Write Position Attribute=true
   )
   &Color(
    Color=&ColorFromCurve(ColorCurve=DataInterface() {propertyValues {
       AlphaCurve {keys [<at 0 0 RCIM_Cubic>,<at 0.1239 1>,<at 0.346 1>,<at 1 0 RCIM_Cubic>]}
       BlueCurve {keys [<at -0.0056 6>,<at 0.2609 0.2>]}
       GreenCurve {keys [<at -0.0056 23.0295>,<at 0.2609 0.7676>]}
       RedCurve {keys [<at -0.0056 30>,<at 0.2609 1>]}
      }})
    Scale Alpha=&FloatFromCurve(FloatCurve=DataInterface() {propertyValues {Curve {keys [<at 0.1282 1.0115>,<at 0.2455 0.0123>,<at 0.3389 0.9565>,<at 0.4758 0.9657>,<at 0.591 0.0032>,<at 0.6388 1.0115>,<at 0.8061 1.0207>,<at 1 0>]}}})
    Scale Color=&VectorFromFloat(Value=&FloatFromCurve(FloatCurve=DataInterface() {propertyValues {Curve {keys [<at 0 2>,<at 0.2994 -0.2627>,<at 0.573 2>,<at 0.8425 0.0123>]}}}))
   )
   &CameraOffset(Camera Offset Amount=25)
   &ScaleRibbonWidth(Ribbon Width Scale=&FloatFromCurve(
     FloatCurve=DataInterface() {propertyValues {Curve {keys [<at 0 0.35 RCIM_Constant>,<at 200 1 RCIM_Constant>]}}}
     CurveIndex=&ClampFloat(Float=$Engine.Owner.LODDistance, Min=20, Max=200)
    ))
   &DynamicMaterialParameters(
    Index 0 Param 1=1
    Index 0 Param 2=1
    Index 0 Param 3=&FloatFromCurve(
     FloatCurve=DataInterface() {propertyValues {Curve {keys [<at 0 2 RCIM_Constant>,<at 200 0.8 RCIM_Constant>]}}}
     CurveIndex=&ClampFloat(Float=$Engine.Owner.LODDistance, Max=200)
    )
    Module.Param0WriteEnabled=false
    Module.Param1WriteEnabled=false
    Module.Param2WriteEnabled=true
    Module.Param3WriteEnabled=false
   ) {module=DynamicMaterialParameters001}
   &ScaleRibbonWidth(Ribbon Width Scale=&MakeCustomFloatFromBool(Bool=$User.bisHandHidden, True Float=0, False Float=1)) {module=ScaleRibbonWidth001}
  ]
  rendererClasses [&NiagaraRibbonRendererProperties]
  renderers [{rendererClass=&NiagaraRibbonRendererProperties}]
 }
 BoneFlow_Light_Left {
  emitterSpawnScript [
   /SetVariables(Emitter.RandomSeed:NiagaraInt32=&RandomInt(Randomness Mode:ENiagaraRandomnessMode=::NewEnumerator0, Random Seed:NiagaraInt32=0, Fixed Random Seed:NiagaraBool=false)) {isSetParameters=true, module=SetVariables_#1}
   &FN_PartitionAttributes(Number of Bone Sets=1)
  ]
  emitterUpdateScript [
   &EmitterState(Life Cycle Mode=::NewEnumerator1, Loop Duration=2, Scalability Mode=::NewEnumerator1, UseLoopDelay=false)
   &SpawnBurst_Instantaneous(
    Spawn Count=$Emitter.FN_PartitionAttributes.Spawn Total
    Spawn Time=&RandomRangeFloat(Minimum=0, Maximum=1)
    Use Spawn Probability=false
   )
  ]
  particleSpawnScript [
   &FN_Particle_Partition_v5(Enable Ribbon ID=true, Seed:NiagaraInt32=$Emitter.RandomSeed)
   &InitializeParticle(
    Lifetime=0.25
    Module.Write Color=true
    Module.Write Lifetime=true
    Module.Write Mass=true
    Module.Write Material Random=true
    Module.Write Position=false
    Module.Write Scale=false
    Module.Write SpriteRotation=true
    Module.Write SpriteSize=true
    Position=&SimulationPosition()
    Sprite Rotation=&UniformRangedFloat(Minimum=0, Maximum=1)
    Sprite Size=&Vector2DFromFloat(Value=10)
   )
  ]
  particleUpdateScript [
   &ParticleState()
   &FN_BoneFlow(
    Angle=0.25
    Height=0.8
    Interpolate Mesh Sampling=false
    Radius=15
    SkeletalMesh:NiagaraDataInterfaceSkeletalMesh=$User.SkeletalMeshLeft
    Write Position Attribute=true
   )
   &FN_Light_v3(
    Color=&ColorFromCurve(ColorCurve=DataInterface() {propertyValues {
       AlphaCurve {keys [<at 0.0007 0>,<at 0.0604 1 RCIM_Cubic>,<at 1 0 RCIM_Cubic>]}
       BlueCurve {keys [<at 0 1.2 RCIM_Cubic>,<at 0.2541 1.2>,<at 0.341 0.0181>,<at 0.4857 1.2>,<at 0.7554 1.2>,<at 0.8188 0.2>,<at 1 1 RCIM_Cubic>]}
       GreenCurve {keys [<at 0 5.1097 RCIM_Cubic>,<at 0.2541 5.1097>,<at 0.341 0.8179>,<at 0.4857 5.1097>,<at 0.7554 5.1097>,<at 0.8188 0.8516>,<at 1 1 RCIM_Cubic>]}
       RedCurve {keys [<at 0 6 RCIM_Cubic>,<at 0.2541 6>,<at 0.341 1>,<at 0.4857 6>,<at 0.7554 6>,<at 0.8188 1>,<at 1 1 RCIM_Cubic>]}
      }})
    Create Color Attribute=true
    Create Exponent Attribute=true
    Create Radius Attribute=true
    Enable Particle Light Enable=true
    Exponent=6
    Light Enabled=true
    Radius=50
    Scale Color=0.5
   )
   /SetVariables(Particles.VisibilityTag=&MakeCustomIntFromBool(Bool=$User.bisHandHidden, True INT=1, False INT=0)) {isSetParameters=true, module=SetVariables_#2}
  ]
  rendererClasses [&NiagaraLightRendererProperties]
  renderers [{rendererClass=&NiagaraLightRendererProperties, rendererIndex=0}]
 }
 BoneFlow_Light_Right {
  emitterSpawnScript [
   /SetVariables(Emitter.RandomSeed:NiagaraInt32=&RandomInt(Randomness Mode:ENiagaraRandomnessMode=::NewEnumerator0, Random Seed:NiagaraInt32=0, Fixed Random Seed:NiagaraBool=false)) {isSetParameters=true, module=SetVariables_#1}
   &FN_PartitionAttributes(Number of Bone Sets=1)
  ]
  emitterUpdateScript [
   &EmitterState(Life Cycle Mode=::NewEnumerator1, Loop Duration=2, Scalability Mode=::NewEnumerator1, UseLoopDelay=false)
   &SpawnBurst_Instantaneous(
    Spawn Count=$Emitter.FN_PartitionAttributes.Spawn Total
    Spawn Time=&RandomRangeFloat(Minimum=0, Maximum=1)
    Use Spawn Probability=false
   )
  ]
  particleSpawnScript [
   &FN_Particle_Partition_v5(Enable Ribbon ID=true, Seed:NiagaraInt32=$Emitter.RandomSeed)
   &InitializeParticle(
    Lifetime=0.25
    Module.Write Color=true
    Module.Write Lifetime=true
    Module.Write Mass=true
    Module.Write Material Random=true
    Module.Write Position=false
    Module.Write Scale=false
    Module.Write SpriteRotation=true
    Module.Write SpriteSize=true
    Position=&SimulationPosition()
    Sprite Rotation=&UniformRangedFloat(Minimum=0, Maximum=1)
    Sprite Size=&Vector2DFromFloat(Value=10)
   )
  ]
  particleUpdateScript [
   &ParticleState()
   &FN_BoneFlow(
    Angle=0.25
    Height=0.8
    Interpolate Mesh Sampling=false
    Radius=15
    SkeletalMesh=DataInterface() {propertyValues {
      DefaultMesh=null
      ExcludeBoneName=root
      OptionalComponentClass=null
      PreviewMesh=&M_Med_Soldier_04_Preview
      SkinningMode=SkinOnTheFly
      SoftSourceActor=null
      SourceMode=Default
      UvSetIndex=0
      WholeMeshLOD=-1
      bExcludeBone=false
      bReadDeformedGeometry=true
      bRequireCurrentFrameData=true
      ComponentTags []
      FilteredBones [spine_05,upperarm_r,lowerarm_r]
      FilteredSockets []
      MeshUserParameter {parameter {name=None, typeDefHandle {registeredTypeIndex=1}, varData []}}
      SamplingRegions []
     }}
    Write Position Attribute=true
   )
   &FN_Light_v3(
    Color=&ColorFromCurve(ColorCurve=DataInterface() {propertyValues {
       AlphaCurve {keys [<at 0.0007 0>,<at 0.0604 1 RCIM_Cubic>,<at 1 0 RCIM_Cubic>]}
       BlueCurve {keys [<at 0 1.2 RCIM_Cubic>,<at 0.2541 1.2>,<at 0.341 0.2>,<at 0.4857 1.2>,<at 0.7554 1.2>,<at 0.8188 0.2>]}
       GreenCurve {keys [<at 0 5.1097 RCIM_Cubic>,<at 0.2541 5.1097>,<at 0.341 0.8516>,<at 0.4857 5.1097>,<at 0.7554 5.1097>,<at 0.8188 0.8516>]}
       RedCurve {keys [<at 0 6 RCIM_Cubic>,<at 0.2541 6>,<at 0.341 1>,<at 0.4857 6>,<at 0.7554 6>,<at 0.8188 1>]}
      }})
    Create Color Attribute=true
    Create Exponent Attribute=true
    Create Radius Attribute=true
    Enable Particle Light Enable=true
    Exponent=6
    Light Enabled=true
    Radius=50
    Scale Color=0.5
   )
   /SetVariables(Particles.VisibilityTag=&MakeCustomIntFromBool(Bool=$User.bisHandHidden, True INT=1, False INT=0)) {isSetParameters=true, module=SetVariables_#2}
  ]
  rendererClasses [&NiagaraLightRendererProperties]
  renderers [{rendererClass=&NiagaraLightRendererProperties, rendererIndex=0}]
 }
 NE_Electric {
  emitterSpawnScript [
   /SetVariables(Emitter.RandomSeed:NiagaraInt32=&RandomInt(Randomness Mode:ENiagaraRandomnessMode=::NewEnumerator0, Random Seed:NiagaraInt32=0, Fixed Random Seed:NiagaraBool=false)) {isSetParameters=true, module=SetVariables_#1}
   &FN_PartitionAttributes(Number of Bone Sets=3, Particles Per Partition=5, Partitions=3)
  ]
  emitterUpdateScript [
   &EmitterState(Life Cycle Mode=::NewEnumerator1, Loop Duration=2, Scalability Mode=::NewEnumerator1, UseLoopDelay=false)
   &SpawnRate(SpawnGroup=3, SpawnRate=5, Use Spawn Probability=false)
  ]
  particleSpawnScript [
   &FN_Particle_Partition_v5(Enable Ribbon ID=true, Pattern:ENiagaraPartitionWaveForm=::NewEnumerator1, Seed:NiagaraInt32=$Emitter.RandomSeed)
   &FN_Initialize_BoneFlow_Attributes_v2(
    Angle:NiagaraFloat=&RandomRangeFloat(Minimum=0, Maximum=1)
    Radius:NiagaraFloat=&RandomRangeFloat(Minimum=0, Maximum=4)
   )
   &InitializeParticle(
    Lifetime=&RandomRangeFloat(Minimum=0.4, Maximum=0.5)
    Module.Write Color=true
    Module.Write Lifetime=true
    Module.Write Mass=true
    Module.Write Material Random=true
    Module.Write Position=false
    Module.Write Scale=false
    Module.Write SpriteRotation=true
    Module.Write SpriteSize=true
    Position=&SimulationPosition()
    Sprite Rotation=&RandomRangeFloat(Minimum=0, Maximum=360)
    Sprite Size=&UniformRangedVector2D(Minimum=<20 15>, Maximum=<35 20>)
   ) {module=InitializeParticle001}
   /SetVariables(Particles.ParentSize:Vector2f=&UniformRangedVector2D(Minimum=<50 40>, Maximum=<70 60>)) {isSetParameters=true}
   &DynamicMaterialParameters(
    Index 0 Param 1=&RandomRangeFloat(Minimum=2.5, Maximum=3.5)
    Index 0 Param 2=&FloatFromCurve(FloatCurve=DataInterface() {propertyValues {Curve {keys [<at 0 1>,<at 1 -1>]}}})
    Index 0 Param 3=&RandomRangeFloat(Minimum=0.3, Maximum=0.6)
    Module.Param0WriteEnabled=true
    Module.Param1WriteEnabled=false
    Module.Param2WriteEnabled=true
    Module.Param3WriteEnabled=false
   )
   &CameraOffset(Camera Offset Amount=&RandomRangeFloat(Minimum=-5, Maximum=25))
   &CameraOffset(Camera Offset Amount=&RandomRangeFloat(Minimum=-5, Maximum=25)) {module=CameraOffset001}
  ]
  particleUpdateScript [
   &ParticleState()
   &FN_BoneFlow(
    Angle:NiagaraFloat=$Particles.Initial.FN_Initialize_BoneFlow_Attributes_v2.Angle
    Height:NiagaraFloat=&Multiply_Float(A=$Particles.Initial.Partition.NormalizedID, B=0.8)
    Interpolate Mesh Sampling=false
    Radius:NiagaraFloat=$Particles.Initial.FN_Initialize_BoneFlow_Attributes_v2.Radius
    SkeletalMesh:NiagaraDataInterfaceSkeletalMesh=$User.SkeletalMesh
    Write Position Attribute=true
   )
   &DynamicMaterialParameters(
    Index 0 Param 1=&RandomRangeFloat(Minimum=2.5, Maximum=3.5)
    Index 0 Param 2=&FloatFromCurve(FloatCurve=DataInterface() {propertyValues {Curve {keys [<at 0 0.8>,<at 1 -1>]}}})
    Index 0 Param 3=&RandomRangeFloat(Minimum=0.2, Maximum=0.7)
    Module.Param0WriteEnabled=false
    Module.Param1WriteEnabled=true
    Module.Param2WriteEnabled=false
    Module.Param3WriteEnabled=false
   ) {module=DynamicMaterialParameters001}
   &Color(
    Color=&ColorFromCurve(ColorCurve=DataInterface() {propertyValues {
       AlphaCurve {keys [<at 0 0 RCIM_Cubic>,<at 0.1239 1>,<at 0.346 1>,<at 1 0 RCIM_Cubic>]}
       BlueCurve {keys [<at -0.0056 1.7969>,<at 0.2101 0.2>]}
       GreenCurve {keys [<at -0.0056 2.6506>,<at 0.2101 0.7676>]}
       RedCurve {keys [<at -0.0056 3>,<at 0.2101 1>]}
      }})
    Scale Alpha=&FloatFromCurve(FloatCurve=DataInterface() {propertyValues {Curve {keys [<at 0.1282 1.0115>,<at 0.2455 0.0123>,<at 0.3389 0.9565>,<at 0.4758 0.9657>,<at 0.591 0.0032>,<at 0.6388 1.0115>,<at 0.8061 1.0207>,<at 1 0>]}}})
    Scale Color=&VectorFromFloat(Value=&FloatFromCurve(FloatCurve=DataInterface() {propertyValues {Curve {keys [<at 0 2>,<at 0.2994 -0.2627>,<at 0.573 2>,<at 0.8425 0.0123>]}}}))
   )
   &ScaleSpriteSize(Scale Factor=&Vector2DFromCurve(Vector2Curve=DataInterface() {propertyValues {XCurve {keys [<at 0 0.3>,<at 1 1>]}, YCurve {keys [<at 0 1>,<at 1 1>]}}}, ScaleCurve=<0.7 0.7>))
   /SetVariables(Particles.ParentColor:LinearColor=<rgba 1 0.8468 0.3591 0.02>) {isSetParameters=true}
   /SetVariables(Particles.VisibilityTag=&MakeCustomIntFromBool(Bool=$User.bisHandHidden, True INT=1, False INT=0)) {isSetParameters=true, module=SetVariables_#2}
  ]
  rendererClasses [&NiagaraSpriteRendererProperties]
  renderers [
   {rendererClass=&NiagaraSpriteRendererProperties}
   {rendererClass=&NiagaraSpriteRendererProperties, rendererIndex=1}
   {rendererClass=&NiagaraSpriteRendererProperties, rendererIndex=2}
  ]
 }
}
summary {
 systemName=NS_LunchBox_Body_Loop4
 emitters [
  {emitterName=NE_Electric, rendererClasses [&NiagaraSpriteRendererProperties]}
  {emitterName=BoneFlow_ArcLarge_Left, rendererClasses [&NiagaraRibbonRendererProperties]}
  {emitterName=BoneFlow_ArcLarge001, rendererClasses [&NiagaraRibbonRendererProperties]}
  {emitterName=BoneFlow_Light_Right, rendererClasses [&NiagaraLightRendererProperties]}
  {emitterName=BoneFlow_Light_Left, rendererClasses [&NiagaraLightRendererProperties]}
 ]
 userVariables [
  User.bisHandHidden {type=NiagaraBool}
  User.RibbonIDOffset {type=NiagaraInt32}
  User.SkeletalMesh {
   type=NiagaraDataInterfaceSkeletalMesh
   defaultValue {
    struct=NiagaraExt_VariableValue_DataInterface
    value {dataInterface=&self:NiagaraDataInterfaceSkeletalMesh_0}
   }
  }
  User.SkeletalMeshLeft {
   type=NiagaraDataInterfaceSkeletalMesh
   defaultValue {
    struct=NiagaraExt_VariableValue_DataInterface
    value {dataInterface=&self:NiagaraDataInterfaceSkeletalMesh_3}
   }
  }
  User.Sphere Origin {type=Vector3f, description="", defaultValue {struct=Vector3f, value=<0 0 0>}}
  User.StatChangedKills {type=NiagaraFloat, defaultValue=0, description=""}
 ]
}
